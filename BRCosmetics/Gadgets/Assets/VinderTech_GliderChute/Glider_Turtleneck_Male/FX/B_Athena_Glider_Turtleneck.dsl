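; Glider Blueprint (Turtleneck) that drives a parachute's trail VFX and audio in gameplay and frontend preview.
; Visibility and full-deploy events attach, toggle and flag two trail components, fade in audio, and manage a preview sound mix.
; Trail updates feed movement values to glider audio and Niagara parameters, with optional legacy Cascade trail and vehicle-voice paths.
; The construction script optionally adds an animated skeletal "Bits" mesh with material overrides.
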(event Parachute|EventOnParachuteTrailUpdated
  (params (MovementLengthSquared :type Float%20%28single-precision%29) (ForwardDot :type Float%20%28single-precision%29) (RightDot :type Float%20%28single-precision%29) (RotationalMovementAcceleration :type Float%20%28single-precision%29))
  ;[1 1 1] Trail Updated - Ticks in Gameplay; Fires only once in Frontend
  (CallFunction|AudioUpdated ForwardDot RightDot)
  (CallFunction|FXUpdated MovementLengthSquared ForwardDot RightDot RotationalMovementAcceleration))

(event Parachute|EventOnReceiveParachuteItemdef
  ;[1 1 1] On Receive Parachute Itemdef
  (Variables|Animation|SetParachute%20Anim%20Instance (Components|SkeletalMesh|GetAnimInstance (Variables|Parachute|GetParachuteMesh)))
  (CallFunction|TrailVFXComponentsAttachment) ; Almost Always Fails as of 18.10
)

(event EventOnChuteVisibilityChanged
  (params (bParachuteVisibility :type Boolean))
  (bind _returnvalue (CallFunction|GetTrailVFXComponent))
  (bind _returnvalue_2 (CallFunction|GetTrailVFXComponent))
  (bind _returnvalue_3 (CallFunction|GetTrailVFX2Component))
  (bind _returnvalue_4 (CallFunction|GetTrailVFXComponent))
  (bind _returnvalue_5 (CallFunction|GetTrailVFX2Component))
  (bind _returnvalue_6 (CallFunction|GetTrailVFX2Component))
  (bind (_asniagara_particle_system_component _bsuccess) (Utilities|Casting|CastToNiagaraComponent _returnvalue_5))
  (bind (_asniagara_particle_system_component_2 _bsuccess_2) (Utilities|Casting|CastToNiagaraComponent _returnvalue_4))
  (bind _bisfrontendpreview (Variables|FortPlayerParachute|GetbIsFrontEndPreview))
  (bind _parachuteitemdef (Variables|FortPlayerParachute|GetParachuteItemDef))
  (bind _parachuteitemdef_2 (Variables|FortPlayerParachute|GetParachuteItemDef))
  ;[1 1 1] Chute Visibility Changed - Fires once on Glider activation and again on deactivation of Glider (fires in Frontend and Gameplay)
  (if bParachuteVisibility
    (CallFunction|TrailVFXComponentsAttachment)
    (if _bisfrontendpreview
      (bind _in_sound_mix_modifier (Variables|Default|GetIn%20Sound%20Mix%20Modifier)
        (:then
          (Audio|PopSoundMixModifier _in_sound_mix_modifier))
        (:else
          _))
      (Audio|PushSoundMixModifier (Variables|Default|GetIn%20Sound%20Mix%20Modifier)))
    (if (and _bisfrontendpreview (CallFunction|IsGliderFullyDeployedAccordingtoAnim))
      (Utilities|IsValid _returnvalue_4
        (:Is%20Valid
          (if (or (Variables|VFX|GetTrailVFX%20Deployed%20Activation) _bsuccess_2)
            (Components|Activation|Activate _returnvalue_4 false)
            (Effects|Components|ParticleSystem|SetBoolParameter _returnvalue_4 "bisFrontEnd" true)
            (Effects|Components|ParticleSystem|SetBoolParameter _returnvalue_4 "bIsFrontEndPreview" true)))
        (:Is%20Not%20Valid
          _))
      (Utilities|IsValid _returnvalue_5
        (:Is%20Valid
          (if (or (Variables|VFX|GetTrailVFX2%20Deployed%20Activation) _bsuccess)
            (Components|Activation|Activate _returnvalue_5 false)
            (Effects|Components|ParticleSystem|SetBoolParameter _returnvalue_5 "bisFrontEnd" true)
            (Effects|Components|ParticleSystem|SetBoolParameter _returnvalue_5 "bIsFrontEndPreview" true)))
        (:Is%20Not%20Valid
          _))
      (for _array_element (Variables|FortPlayerParachute|GetTrailComponents)
        (Utilities|IsValid _array_element
          (:Is%20Valid
            (Effects|Components|ParticleSystem|SetBoolParameter _array_element "bisFrontEnd" true)
            (Effects|Components|ParticleSystem|SetBoolParameter _array_element "bIsFrontEndPreview" true))))
      (elif (Variables|VFX|GetAuto%20Activate%20Fix)
        (Utilities|IsValid _returnvalue
          (:Is%20Valid
            (Components|Activation|Deactivate _returnvalue)
            (if (Athena|GetAttachNiagaraEffecttoPlayerPawn _parachuteitemdef_2)
              (Niagara|SetNiagaraSkeletalMeshComponent (Utilities|Casting|CastToNiagaraComponent (CallFunction|GetTrailVFXComponent)) (Utilities|String|ToString(Name) (Athena|GetUserSkeletonParameterName _parachuteitemdef_2)) (Variables|/Script/Engine.Character|GetMesh (CallFunction|GetFortPlayerPawn)))))
          (:Is%20Not%20Valid
            _))
        (Utilities|IsValid _returnvalue_6
          (:Is%20Valid
            (Components|Activation|Deactivate _returnvalue_6)
            (if (Athena|GetAttachNiagaraEffecttoPlayerPawn _parachuteitemdef)
              (Niagara|SetNiagaraSkeletalMeshComponent (Utilities|Casting|CastToNiagaraComponent (CallFunction|GetTrailVFX2Component)) (Utilities|String|ToString(Name) (Athena|GetUserSkeletonParameterName _parachuteitemdef)) (Variables|/Script/Engine.Character|GetMesh (CallFunction|GetFortPlayerPawn))))))
        (else
          (Components|Activation|Activate (CallFunction|GetTrailVFXComponent) true))))
    (else
      (Utilities|IsValid _returnvalue_2
        (:Is%20Valid
          (Effects|Components|ParticleSystem|SetBoolParameter _returnvalue_2 "bisFullyDeployed" false)
          (Components|Activation|Deactivate _returnvalue_2))
        (:Is%20Not%20Valid
          _))
      (Utilities|IsValid _returnvalue_3
        (:Is%20Valid
          (Effects|Components|ParticleSystem|SetBoolParameter _returnvalue_3 "bisFullyDeployed" false)
          (Components|Activation|Deactivate _returnvalue_3))
        (:Is%20Not%20Valid
          _))
      ;; Legacy VFX Trail Setup
      (if (Variables|LegacySetup|GetUseLegacyFXSetup)
        (if (Variables|LegacyVFX|GetLegacy%20Toggle%20TrailVFX2%20Activation%20Order)
          (Variables|LegacyVFX|SetLegacy%20DoOnce_TrailVFXInverted true)))
      (Variables|Audio|SetbAudioHasFadedIn false)
      (for _array_element_2 (Variables|FortPlayerParachute|GetTrailComponents)
        (Utilities|IsValid _array_element_2
          (:Is%20Valid
            (Effects|Components|ParticleSystem|SetBoolParameter _array_element_2 "bisFullyDeployed" false)
            (Components|Activation|Deactivate _array_element_2))))
      (for _array_element_3 (Components|GetChildrenComponents (Variables|Parachute|GetParachuteMesh) true)
        (bind _asfxsystem_component (Utilities|Casting|CastToFXSystemComponent _array_element_3))
        (bind _array_element_4 (Utilities|Array|ForEachLoopwithBreak (Variables|/Script/Engine.ActorComponent|GetComponentTags _asfxsystem_component) :!label _label)
          (:LoopBody
            (if (Utilities|Name|Equal(Name) _array_element_4 "GliderDeploy")
              (Components|Activation|Deactivate _asfxsystem_component)
              (goto _label Break))))))))

(event EventOnGliderFullyDeployed
  (bind _returnvalue (CallFunction|GetTrailVFX2Component))
  (bind _returnvalue_2 (CallFunction|GetTrailVFXComponent))
  (bind (_ascascade_particle_system_component _bsuccess) (Utilities|Casting|CastToParticleSystemComponent _returnvalue_2))
  (bind (_ascascade_particle_system_component_2 _bsuccess_2) (Utilities|Casting|CastToParticleSystemComponent _returnvalue))
  (bind (_asniagara_particle_system_component _bsuccess_3) (Utilities|Casting|CastToNiagaraComponent _returnvalue))
  (bind (_asniagara_particle_system_component_2 _bsuccess_4) (Utilities|Casting|CastToNiagaraComponent _returnvalue_2))
  (bind _bisfrontendpreview (Variables|FortPlayerParachute|GetbIsFrontEndPreview))
  (bind _bisfrontendpreview_2 (Variables|FortPlayerParachute|GetbIsFrontEndPreview))
  ;[1 1 1] Fully Deployed - Fires on TurnONEngines Anim Notify position in the INTO animations (will fire in Frontend and Gameplay)
  (if (not (Variables|LegacySetup|GetUseLegacyAudioSetup))
    (Audio|Components|Audio|FadeIn (Variables|Parachute|GetFortLayeredAudioComponentGlider) (Variables|Audio|GetAudioFadeInTime) 1.0 0.0 "Linear" :!class /Script/Engine.AudioComponent)
    (Variables|Audio|SetTimeAudioFadeStarted (Utilities|Time|GetGameTimeinSeconds)))
  (Utilities|IsValid _returnvalue_2
    (:Is%20Valid
      (if (or _bsuccess_4 (Variables|VFX|GetTrailVFX%20Deployed%20Activation))
        (Components|Activation|Activate _returnvalue_2 false)
        ;; Legacy Trail VFX Deployed Setup
        (if (Variables|LegacySetup|GetUseLegacyFXSetup)
          (if _bsuccess
            (Effects|Particles|Trails|BeginTrails _ascascade_particle_system_component (Variables|LegacyVFX|GetVFX%20Component%201st%20Socket%20Name) (Variables|LegacyVFX|GetVFX%20Component%202nd%20Socket%20Name) (Variables|LegacyVFX|GetVFX%20Component%20Width%20Mode) (Variables|LegacyVFX|GetVFX%20Component%20Width))))
        (Effects|Components|ParticleSystem|SetBoolParameter _returnvalue_2 "bisFullyDeployed" true)
        (Effects|Components|ParticleSystem|SetBoolParameter _returnvalue_2 "bisFrontEnd" _bisfrontendpreview_2)
        (Effects|Components|ParticleSystem|SetBoolParameter _returnvalue_2 "bIsFrontEndPreview" _bisfrontendpreview_2)))
    (:Is%20Not%20Valid
      _))
  (Utilities|IsValid _returnvalue
    (:Is%20Valid
      (if (or _bsuccess_3 (Variables|VFX|GetTrailVFX2%20Deployed%20Activation))
        (Components|Activation|Activate _returnvalue false)
        ;; Legacy VFX2 Trail Deployed Setup
        (if (Variables|LegacySetup|GetUseLegacyFXSetup)
          (if _bsuccess_2
            (Effects|Particles|Trails|BeginTrails _ascascade_particle_system_component_2 (Variables|LegacyVFX|GetVFX2%20Component%201st%20Socket%20Name) (Variables|LegacyVFX|GetVFX2%20Component%202nd%20Socket%20Name) (Variables|LegacyVFX|GetVFX2%20Component%20Width%20Mode) (Variables|LegacyVFX|GetVFX2%20Component%20Width))))
        (Effects|Components|ParticleSystem|SetBoolParameter _returnvalue "bisFullyDeployed" true)
        (Effects|Components|ParticleSystem|SetBoolParameter _returnvalue "bisFrontEnd" _bisfrontendpreview)
        (Effects|Components|ParticleSystem|SetBoolParameter _returnvalue "bIsFrontEndPreview" _bisfrontendpreview)))))

(event EventDestroyed
  ;[1 1 1] Remove mix mod when not needed
  (bind _in_sound_mix_modifier (Variables|Default|GetIn%20Sound%20Mix%20Modifier)
    (:then
      (Audio|PopSoundMixModifier _in_sound_mix_modifier))
    (:else)))

(event EventEndPlay
  (params (EndPlayReason :type EEndPlayReason%20Enum :default "Destroyed"))
  ;[1 1 1] Remove mix mod when not needed
  (bind _in_sound_mix_modifier (Variables|Default|GetIn%20Sound%20Mix%20Modifier)
    (:then
      (Audio|PopSoundMixModifier _in_sound_mix_modifier))
    (:else)))

(fn AudioFadeCheck
  (return (> (- (Utilities|Time|GetGameTimeinSeconds) (Variables|Audio|GetTimeAudioFadeStarted)) (Variables|Audio|GetAudioFadeInTime))))

(fn AudioUpdated
  (params (Forward%20Dot :type Float%20%28double-precision%29) (Right%20Dot :type Float%20%28double-precision%29))
  (bind _fortvehicleaudiovoice (Variables|LegacyAudio|GetFortVehicleAudioVoice))
  (if (Variables|LegacySetup|GetUseLegacyAudioSetup)
    (if (CallFunction|IsGliderFullyDeployedAccordingtoAnim)
      (Class|FortVehicleAudioVoice|SetParam _fortvehicleaudiovoice (select (>= (select (Variables|LegacyAudio|GetAbsolute%20Movement_Audio) (Math|Float|Absolute(Float) (Variables|GetForward%20Dot)) (Variables|GetForward%20Dot)) 0.2) (Variables|LegacyAudio|GetMax%20Audio%20Voice) (Variables|LegacyAudio|GetMin%20Audio%20Voice)))
      (else
        (Class|FortVehicleAudioVoice|SetParam _fortvehicleaudiovoice (Variables|LegacyAudio|GetLegacy%20Not%20Deployed%20Audio%20Voice))))
    (elif (Variables|Audio|GetbAudioHasFadedIn)
      _
      (else
        (bind _output_get (Variables|Audio|SetbAudioHasFadedIn (CallFunction|AudioFadeCheck)))
        (if _output_get
          _))))
  (Class|FortGliderAudioComponent|UpdateGliderAudio (Variables|Parachute|GetFortLayeredAudioComponentGlider) Forward%20Dot Right%20Dot))

(fn FXUpdated
  (params (Movement :type Float%20%28double-precision%29) (Forward%20Dot :type Float%20%28double-precision%29) (Right%20Dot :type Float%20%28double-precision%29) (Rotational%20Movement :type Float%20%28double-precision%29))
  (bind _returnvalue (CallFunction|GetTrailVFXComponent))
  (bind _returnvalue_2 (CallFunction|GetTrailVFX2Component))
  (bind _returnvalue_3 (CallFunction|GetTrailVFX2Component))
  (bind _returnvalue_4 (CallFunction|GetTrailVFXComponent))
  (bind _returnvalue_5 (CallFunction|IsGliderFullyDeployedAccordingtoAnim))
  (bind _returnvalue_6 (CallFunction|GetTrailVFX2Component))
  (bind _returnvalue_7 (CallFunction|GetTrailVFXComponent))
  (bind _returnvalue_8 (CallFunction|GetTrailVFX2Component))
  (bind _returnvalue_9 (CallFunction|GetTrailVFX2Component))
  (bind (_ascascade_particle_system_component _bsuccess) (Utilities|Casting|CastToParticleSystemComponent _returnvalue_4))
  (bind (_ascascade_particle_system_component_2 _bsuccess_2) (Utilities|Casting|CastToParticleSystemComponent _returnvalue_7))
  (bind (_ascascade_particle_system_component_3 _bsuccess_3) (Utilities|Casting|CastToParticleSystemComponent _returnvalue_8))
  (bind (_ascascade_particle_system_component_4 _bsuccess_4) (Utilities|Casting|CastToParticleSystemComponent _returnvalue_9))
  (bind (_ascascade_particle_system_component_5 _bsuccess_5) (Utilities|Casting|CastToParticleSystemComponent _returnvalue_3))
  (bind (_ascascade_particle_system_component_6 _bsuccess_6) (Utilities|Casting|CastToParticleSystemComponent _returnvalue_6))
  (bind (_asniagara_particle_system_component _bsuccess_7) (Utilities|Casting|CastToNiagaraComponent _returnvalue_2))
  (bind (_asniagara_particle_system_component_2 _bsuccess_8) (Utilities|Casting|CastToNiagaraComponent _returnvalue))
  (if (Variables|LegacySetup|GetUseLegacyFXSetup)
    ; Legacy VFX Setup
    (if (and (or (>= (select (Variables|LegacyVFX|GetAbsolute%20Movement%20FX) (Math|Float|Absolute(Float) (Variables|GetForward%20Dot)) (Variables|GetForward%20Dot)) 0.2) (Variables|LegacyVFX|GetAdditional%20Condition%20for%20Movement%20Activation)) (CallFunction|IsGliderFullyDeployedAccordingtoAnim))
      (if (Variables|LegacyVFX|GetLegacy%20DoOnce)
        (if (not (Variables|VFX|GetTrailVFX%20Deployed%20Activation))
          (Utilities|IsValid _returnvalue_7
            (:Is%20Valid
              (Components|Activation|Activate _returnvalue_7 false)
              (if _bsuccess_2
                (Effects|Particles|Trails|BeginTrails _ascascade_particle_system_component_2 (Variables|LegacyVFX|GetVFX%20Component%201st%20Socket%20Name) (Variables|LegacyVFX|GetVFX%20Component%202nd%20Socket%20Name) (Variables|LegacyVFX|GetVFX%20Component%20Width%20Mode) (Variables|LegacyVFX|GetVFX%20Component%20Width))))
            (:Is%20Not%20Valid
              _)))
        (if (Variables|VFX|GetTrailVFX2%20Deployed%20Activation)
          _
          (elif (Variables|LegacyVFX|GetLegacy%20Toggle%20TrailVFX2%20Activation%20Order)
            (Utilities|IsValid _returnvalue_8
              (:Is%20Valid
                (Components|Activation|Deactivate _returnvalue_8)
                (if _bsuccess_3
                  (Effects|Particles|Trails|EndTrails _ascascade_particle_system_component_3))
                (Variables|LegacyVFX|SetLegacy%20DoOnce_TrailVFXInverted true))
              (:Is%20Not%20Valid
                _))
            (else
              (Utilities|IsValid _returnvalue_3
                (:Is%20Valid
                  (Components|Activation|Activate _returnvalue_3 false)
                  (if _bsuccess_5
                    (Effects|Particles|Trails|BeginTrails _ascascade_particle_system_component_5 (Variables|LegacyVFX|GetVFX2%20Component%201st%20Socket%20Name) (Variables|LegacyVFX|GetVFX2%20Component%202nd%20Socket%20Name) (Variables|LegacyVFX|GetVFX2%20Component%20Width%20Mode) (Variables|LegacyVFX|GetVFX2%20Component%20Width))))
                (:Is%20Not%20Valid
                  _)))))
        (Variables|LegacyVFX|SetLegacy%20DoOnce false))
      (elif (Variables|VFX|GetTrailVFX%20Deployed%20Activation)
        _
        (else
          (Utilities|IsValid _returnvalue_4
            (:Is%20Valid
              (Components|Activation|Deactivate _returnvalue_4)
              (if _bsuccess
                (Effects|Particles|Trails|EndTrails _ascascade_particle_system_component)))
            (:Is%20Not%20Valid
              _)))))
    (if (Variables|VFX|GetTrailVFX2%20Deployed%20Activation)
      _
      (elif (Variables|LegacyVFX|GetLegacy%20Toggle%20TrailVFX2%20Activation%20Order)
        (if (and (Variables|LegacyVFX|GetLegacy%20DoOnce_TrailVFXInverted) _returnvalue_5)
          (Utilities|IsValid _returnvalue_9
            (:Is%20Valid
              (Components|Activation|Activate _returnvalue_9 false)
              (if _bsuccess_4
                (Effects|Particles|Trails|BeginTrails _ascascade_particle_system_component_4 (Variables|LegacyVFX|GetVFX2%20Component%201st%20Socket%20Name) (Variables|LegacyVFX|GetVFX2%20Component%202nd%20Socket%20Name) (Variables|LegacyVFX|GetVFX2%20Component%20Width%20Mode) (Variables|LegacyVFX|GetVFX2%20Component%20Width)))
              (Variables|LegacyVFX|SetLegacy%20DoOnce_TrailVFXInverted false))
            (:Is%20Not%20Valid
              _))
          (elif _returnvalue_5
            _
            (else
              (Utilities|IsValid _returnvalue_6 :!label _label
                (:Is%20Valid
                  (Components|Activation|Deactivate _returnvalue_6)
                  (if _bsuccess_6
                    (Effects|Particles|Trails|EndTrails _ascascade_particle_system_component_6)))
                (:Is%20Not%20Valid
                  _)))))
        (else
          (goto _label))))
    (Variables|LegacyVFX|SetLegacy%20DoOnce true)
    (else
      (Utilities|IsValid _returnvalue
        (:Is%20Valid
          (if _bsuccess_8
            (Effects|Components|ParticleSystem|SetFloatParameter _returnvalue "Movement" (Variables|GetMovement))
            (Effects|Components|ParticleSystem|SetFloatParameter _returnvalue "ForwardDot" (Variables|GetForward%20Dot))
            (Effects|Components|ParticleSystem|SetFloatParameter _returnvalue "RightDot" (Variables|GetRight%20Dot))
            (Effects|Components|ParticleSystem|SetFloatParameter _returnvalue "RotationalMovement" (Variables|GetRotational%20Movement))
            (else
              (Development|PrintString "Trail VFX Component is NOT Niagara Component" false false))))
        (:Is%20Not%20Valid
          (Development|PrintString "TrailVFX Component NOT VALID" false false)))
      (Utilities|IsValid _returnvalue_2
        (:Is%20Valid
          (if _bsuccess_7
            (Effects|Components|ParticleSystem|SetFloatParameter _returnvalue_2 "Movement" (Variables|GetMovement))
            (Effects|Components|ParticleSystem|SetFloatParameter _returnvalue_2 "ForwardDot" (Variables|GetForward%20Dot))
            (Effects|Components|ParticleSystem|SetFloatParameter _returnvalue_2 "RightDot" (Variables|GetRight%20Dot))
            (Effects|Components|ParticleSystem|SetFloatParameter _returnvalue_2 "RotationalMovement" (Variables|GetRotational%20Movement))
            (else
              (Development|PrintString "Trail VFX2 Component is NOT Niagara Component" false false))))
        (:Is%20Not%20Valid
          (Development|PrintString "TrailVFX2 Component NOT VALID" false false))))))

(fn TrailVFXComponentsAttachment
  (bind _returnvalue (CallFunction|GetTrailVFXComponent))
  (bind _returnvalue_2 (CallFunction|GetTrailVFX2Component))
  (Utilities|IsValid _returnvalue
    (:Is%20Valid
      (Transformation|AttachComponentToComponent _returnvalue (Variables|Parachute|GetParachuteMesh) (Variables|VFX|GetTrailVFX%20Component%20Socket%20Name) (Variables|VFX|GetTrail%20VFX%20Location%20Rule) (Variables|VFX|GetTrail%20VFX%20Rotation%20Rule) (Variables|VFX|GetTrail%20VFX%20Scale%20Rule) true)
      (Transformation|SetRelativeTransform (CallFunction|GetTrailVFXComponent) (Variables|VFX|GetTrailVFX%20Component%20Relative%20Transform) false false))
    (:Is%20Not%20Valid
      _))
  (Utilities|IsValid _returnvalue_2
    (:Is%20Valid
      (Transformation|AttachComponentToComponent _returnvalue_2 (Variables|Parachute|GetParachuteMesh) (Variables|VFX|GetTrailVFX2%20Component%20Socket%20Name) (Variables|VFX|GetTrail%20VFX2%20Location%20Rule) (Variables|VFX|GetTrail%20VFX2%20Rotation%20Rule) (Variables|VFX|GetTrail%20VFX2%20Scale%20Rule) true)
      (Transformation|SetRelativeTransform (CallFunction|GetTrailVFX2Component) (Variables|VFX|GetTrailVFX2%20Component%20Relative%20Transform) false false))
    (:Is%20Not%20Valid
      _))
  (return))

(fn ConstructionScript
  (if (Variables|Bits|GetAdd%20Skeletal%20Mesh%20Bits%20to%20Glider)
    (bind _returnvalue (AddComponent|NODE_AddSkeletalMeshComponentBudgeted-2 :RelativeTransform (Math|Transform|MakeTransform (Variables|Bits|GetBits%20Relative%20Location) (Variables|Bits|GetBits%20Relative%20Rotation) (Variables|Bits|GetBits%20Relative%20Scale))))
    (bind _output_get (Variables|Bits|SetBits _returnvalue))
    (Components|SkinnedMesh|SetSkinnedAssetandUpdate _output_get (Variables|Bits|GetBits%20Skeletal%20Mesh) false)
    (Components|SkeletalMesh|SetAnimInstanceClass _output_get (Variables|Bits|GetBits%20Anim%20BP%20Class))
    (Transformation|AttachComponentToComponent _output_get (Variables|Parachute|GetParachuteMesh) "None" "KeepRelative" "KeepRelative" "KeepRelative" true)
    (if (Variables|Bits|GetBits%20Override%20Materials)
      (for (_ _array_index) (Rendering|Material|GetMaterials (Variables|Bits|GetBits))
        (bind _result (ArrayMacros|SafeArrayGet (Variables|Bits|GetBits%20Materials) _array_index)
          (:%20%20
            (Rendering|Material|SetMaterial (Variables|Bits|GetBits) _array_index _result)))))))
