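# Unlit red beam FX material for the Turtleneck Male glider, flagged for Niagara sprites, ribbons and mesh particles.
# MF_Fresnel_FX, offset by a constant, blends outer and inner red color parameters, with the inner color overbright.
# The scaled result is tinted by particle color and passed through MF_FX_EmissiveAdjustments.
# A separate chain converts bounds-scaled vertex normals to tangent space, with no visible consumer.
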Material /BRCosmetics/Gadgets/Assets/VinderTech_GliderChute/Glider_Turtleneck_Male/FX/M_Red_Beam
graph {
 properties {
  bCastRayTracedShadows=false
  bUsedWithNiagaraMeshParticles=true
  bUsedWithNiagaraRibbons=true
  bUsedWithNiagaraSprites=true
  pixelDepthOffsetMode=PDOM_Legacy
  shadingModel=MSM_Unlit
  usedShadingModels=Unlit
 }
 nodes [
  n1 = ScalarParameter() {defaultValue=1, parameterName=Fresnel Brightness A}
  n2 = ScalarParameter() {defaultValue=8, parameterName=Fresnel Exponent}
  n3 = LinearInterpolate(
   A=VectorParameter() {parameterName=Color Outer, defaultValue=<rgba 0.75 0 0 0>}.RGB
   B=VectorParameter() {parameterName=Color Inner, defaultValue=<rgba 5 0 0.15 0.5>}.RGB
   Alpha=Add(A=MaterialFunctionCall(Brightness=@n1, Exponent=@n2) {materialFunction=&MF_Fresnel_FX}.Fresnel) {constB=0.25}
  )
  MaterialFunctionCall(Emissive=Multiply(A=Multiply(A=@n3) {constB=0.75}, B=ParticleColor().RGB)) {materialFunction=&MF_FX_EmissiveAdjustments}
  n4 = VertexNormalWS()
  n5 = ObjectBounds()
  Transform(None=Normalize(VectorInput=Divide(A=@n4, B=@n5))) {transformSourceType=TRANSFORMSOURCE_World, transformType=TRANSFORM_Tangent}
 ]
}
